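# Editor utility Blueprint for tech artists that applies UI texture settings to locker and featured item textures selected in the Content Browser.
# Two functions each run an embedded Python script that sets no mipmaps, clamp addressing, UI group and BC7, then saves.
# The featured items function also sets per-platform downscale on every texture.
# The locker icons function only processes textures 128 or 256 wide, adding Switch and Mobile downscale for the larger size.
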EditorUtilityBlueprint /Game/Blueprints/Blutilities/ForTechArt/Actions/LockerItemTexture_Actions
graphs [
 EventGraph {links [], nodes []}
 Update TA Featured Items {nodes [
   Update TA Featured Items() {is=title K2Node_FunctionEntry}
   > Execute Python Script(PythonScript="import unreal\r\n\r\nasset_subsys = unreal.get_editor_subsystem(unreal.EditorAssetSubsystem)\r\nselected_textures = unreal.EditorUtilityLibrary.get_selected_asset_data() ###### load selected  hid from content browser ######\r\n\r\nfor tex in selected_textures:\r\n    print(tex.package_name)\r\n    load_selected_TA = asset_subsys.load_asset(tex.package_name)\r\n    downscale_opt = {'Switch': 4.0, 'Mobile': 4.0, 'PS4': 2.0, 'XB1': 2.0, 'Switch2': 2.0}\r\n    downscale = load_selected_TA.get_editor_property('downscale')\r\n    downscale.set_editor_property('per_platform', downscale_opt)\r\n    load_selected_TA.set_editor_property('mip_gen_settings', unreal.TextureMipGenSettings.TMGS_NO_MIPMAPS)\r\n    load_selected_TA.set_editor_property('address_x', unreal.TextureAddress.TA_CLAMP)\r\n    load_selected_TA.set_editor_property('address_y', unreal.TextureAddress.TA_CLAMP)\r\n    load_selected_TA.set_editor_property('lod_group', unreal.TextureGroup.TEXTUREGROUP_UI)\r\n    load_selected_TA.set_editor_property('compression_settings', unreal.TextureCompressionSettings.TC_BC7)\r\n    asset_subsys.save_asset(tex.package_name)")
  ]}
 Update TA Locker Icons {nodes [
   Update TA Locker Icons() {is=title K2Node_FunctionEntry}
   > Execute Python Script(PythonScript="import unreal\r\n\r\nasset_subsys = unreal.get_editor_subsystem(unreal.EditorAssetSubsystem)\r\nselected_textures = unreal.EditorUtilityLibrary.get_selected_asset_data() ###### load selected from content browser ######\r\n\r\nfor tex in selected_textures:\r\n    print(tex.package_name)\r\n    load_selected_TA = asset_subsys.load_asset(tex.package_name)\r\n    size_x = load_selected_TA.blueprint_get_size_x() \r\n    \r\n    if size_x == 128:\r\n        print(\"here\")\r\n        ##downscale_opt = {'Switch': 2.0, 'Mobile': 2.0}\r\n        ##downscale = load_selected_TA.get_editor_property('downscale')\r\n        ##downscale.set_editor_property('per_platform', downscale_opt)\r\n        load_selected_TA.set_editor_property('mip_gen_settings', unreal.TextureMipGenSettings.TMGS_NO_MIPMAPS)\r\n        load_selected_TA.set_editor_property('address_x', unreal.TextureAddress.TA_CLAMP)\r\n        load_selected_TA.set_editor_property('address_y', unreal.TextureAddress.TA_CLAMP)\r\n        load_selected_TA.set_editor_property('lod_group', unreal.TextureGroup.TEXTUREGROUP_UI)\r\n        load_selected_TA.set_editor_property('compression_settings', unreal.TextureCompressionSettings.TC_BC7)\r\n        asset_subsys.save_asset(tex.package_name)\r\n        \r\n    if size_x == 256:\r\n        print(\"here\")\r\n        downscale_opt = {'Switch': 2.0, 'Mobile': 2.0}\r\n        downscale = load_selected_TA.get_editor_property('downscale')\r\n        downscale.set_editor_property('per_platform', downscale_opt)\r\n        load_selected_TA.set_editor_property('mip_gen_settings', unreal.TextureMipGenSettings.TMGS_NO_MIPMAPS)\r\n        load_selected_TA.set_editor_property('address_x', unreal.TextureAddress.TA_CLAMP)\r\n        load_selected_TA.set_editor_property('address_y', unreal.TextureAddress.TA_CLAMP)\r\n        load_selected_TA.set_editor_property('lod_group', unreal.TextureGroup.TEXTUREGROUP_UI)\r\n        load_selected_TA.set_editor_property('compression_settings', unreal.TextureCompressionSettings.TC_BC7)\r\n        asset_subsys.save_asset(tex.package_name)")
  ]}
]
variables []
