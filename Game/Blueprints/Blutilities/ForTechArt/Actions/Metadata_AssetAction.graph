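# Editor utility blueprint that adds an editMetadata asset action for editing the metadata of a selected asset.
# It takes one asset from the current selection (default index) and passes it into an embedded Python script.
# The script spawns the AssetMetadataManager_EWidget editor utility tab and calls its SetCurrentAsset method with that asset, using Python to bridge Blueprint and widget.
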EditorUtilityBlueprint /Game/Blueprints/Blutilities/ForTechArt/Actions/Metadata_AssetAction
graphs [
 EventGraph {links [], nodes []}
 editMetadata {nodes [
   editMetadata() {is=title K2Node_FunctionEntry}
   > n1 = GetSelectedAssets()
   > Execute Python Script!(
    execute
    PythonScript="import unreal\r\n\r\nasset_path_string = \"/Game/Blueprints/Blutilities/ForTechArt/AssetTools/AssetMetadataManager/AssetMetadataManager_EWidget\"\r\nasset = unreal.load_asset(asset_path_string)\r\nss = unreal.get_editor_subsystem(unreal.EditorUtilitySubsystem)\r\ntab = ss.spawn_and_register_tab(asset)\r\ntab.call_method(\"SetCurrentAsset\", (asset_in,))\r\n"
    asset_in:Object Reference="Get (a copy)"(Array:Array of Object References=@n1) -> (Output:Object Reference)
   ) -> (then, ReturnValue) {inputs [asset_in]}
  ]}
]
variables []
