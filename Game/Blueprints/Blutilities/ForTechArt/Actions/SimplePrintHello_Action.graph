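# SimplePrintHello_Action is a tech-art Editor Utility Blueprint that serves as a simple hello-message action in the editor.
# Its Run event prints "Hello" as a string and then as text.
# It then opens an OK-only warning message dialog, using the NewVar variable, which defaults to "Hello", as both title and message.
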EditorUtilityBlueprint /Game/Blueprints/Blutilities/ForTechArt/Actions/SimplePrintHello_Action
graphs [EventGraph {nodes [
   Event Run()
   > PrintString(InString=Hello)
   > PrintText(InText="INVTEXT(\"Hello\")")
   n1 = Get NewVar() -> (NewVar:Text)
   > Show Message Dialog(Title=@n1, Message=@n1, MessageType=Ok, DefaultValue=No, MessageCategory=Warning)
  ]}]
variables [NewVar {type=string, default=Hello}]
