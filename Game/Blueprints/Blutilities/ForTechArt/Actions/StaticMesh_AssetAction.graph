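# Editor utility for tech artists that disables collision, shadows or Nanite on the static meshes selected as assets.
# Internal helpers remove collision or disable section shadow casting across every LOD and section, also zeroing LOD 0's distance field scale for shadows.
# The combined action shows an options dialog persisted through a SaveGame, then either calls those helpers, rescales distance fields, or uses inline Python to set simple-as-complex collision.
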EditorUtilityBlueprint /Game/Blueprints/Blutilities/ForTechArt/Actions/StaticMesh_AssetAction
graphs [
 DisableAllCollision {nodes [
   DisableAllCollision() {is=title K2Node_FunctionEntry}
   > n1 = GetSelectedAssets()
   > n2 = Filter by Class!(
    execute
    TargetArray:Array of Object References=@n1
    ObjectClass:Object Class Reference=&StaticMesh
    FilterType:EEditorScriptingFilterType Enum
   ) -> (then, ReturnValue:Array of Static Mesh Object References)
   > n3 = For Each Loop(Array:Array of Static Mesh Object References=@n2) -> (Array Element:Static Mesh Object Reference)
   > internal_disableCollision!(execute, self, staticMesh:Static Mesh Object Reference=@n3) -> (then)
  ]}
 DisableAllShadows {nodes [
   DisableAllShadows() {is=title K2Node_FunctionEntry}
   > n1 = GetSelectedAssets()
   > n2 = Filter by Class(TargetArray=@n1, ObjectClass=&StaticMesh)
   > n3 = For Each Loop(Array=@n2)
   > internal_disableShadows!(execute, self, staticMesh:Static Mesh Object Reference=@n3) -> (then)
  ]}
 DisableCollisionAndShadows {nodes [
   DisableCollisionAndShadows() {is=title K2Node_FunctionEntry}
   > n1 = Sequence()
   n2 = Get l_saveGameName() -> (l_saveGameName:String)
   > n3 = DoesSaveGameExist!(execute, SlotName:String=@n2, UserIndex:int=0) -> (then, ReturnValue:bool)
   > n4 = Branch(Condition=@n3)
   > n5 = LoadGameFromSlot!(execute, SlotName:String=@n2, UserIndex:int=0) -> (then, ReturnValue:Save Game Object Reference)
   > n6 = Cast To SaveGame_DisableCollisionAndShadows!(execute, Object=@n5) -> (then, CastFailed, AsSave Game Disable Collision and Shadows:Save Game Disable Collision And Shadows Object Reference) {is=title K2Node_DynamicCast}
   > n7 = Set l_saveGame(l_saveGame:Save Game Disable Collision And Shadows Object Reference=@n6) -> (Output_Get:Save Game Disable Collision And Shadows Object Reference)
   n8 = Get l_saveGame() -> (l_saveGame:Save Game Disable Collision And Shadows Object Reference)
   Is Valid(exec=[@n7.then, @n13.then], InputObject=@n8)
   > n9 = Show Object Dialog!(
    execute
    Title:"Text (by ref)"=MakeLiteralText(Value="INVTEXT(\"Disable Shadows and Collision Settings\")")
    InOutObject:Object Reference=@n8
    Options:"Editor Dialog Library Object Details View Options Structure (by ref)"=Make Editor Dialog Library Object Details View Options!(
     bShowObjectName:bool=false
     bAllowSearch:bool=false
     bAllowResizing:bool=False
     MinWidth:int=200
     MinHeight:int=100
     ValueColumnWidthRatio:float=0.6
    ) -> (EditorDialogLibraryObjectDetailsViewOptions:Editor Dialog Library Object Details View Options Structure) {is=title K2Node_MakeStruct}
   ) -> (then, ReturnValue:bool)
   > n10 = Branch(Condition=@n9)
   n11 = Get l_saveGame()
   > Set l_shadowOptions(l_shadowOptions:ShadowDisableOptions Enum=Get ShadowSettings!(self:Save Game Disable Collision And Shadows Object Reference=@n11) -> (ShadowSettings:ShadowDisableOptions Enum=NewEnumerator0)) -> (Output_Get:ShadowDisableOptions Enum=NewEnumerator0)
   > Set l_collisionOptions(l_collisionOptions:CollisionDisableOptions Enum=Get CollisionSettings!(self:Save Game Disable Collision And Shadows Object Reference=@n11) -> (CollisionSettings:CollisionDisableOptions Enum=NewEnumerator0)) -> (Output_Get:CollisionDisableOptions Enum=NewEnumerator0)
   > Set l_scale(l_scale:double=Get DistanceFieldScale!(self:Save Game Disable Collision And Shadows Object Reference=@n11) -> (DistanceFieldScale:double)) -> (Output_Get:double)
   > SaveGameToSlot!(execute, SaveGameObject:Save Game Object Reference=Get l_saveGame(), SlotName:String=Get l_saveGameName(), UserIndex:int=0) -> (then, ReturnValue:bool)
   Return Node(execute=@n10.else)
   n12 = CreateSaveGameObject!(
    execute=@n4.else
    SaveGameClass:Save Game Class Reference=&/SaveGame_DisableCollisionAndShadows.SaveGame_DisableCollisionAndShadows_C
   ) -> (then, ReturnValue:Save Game Disable Collision And Shadows Object Reference)
   > n13 = Set l_saveGame(l_saveGame=@n12)
   n14 = GetSelectedAssets(execute=@n1.then_1)
   > n15 = Filter by Class(TargetArray=@n14, ObjectClass=&StaticMesh)
   > n16 = For Each Loop(Array=@n15)
   > n17 = Sequence()
   > n18 = Switch on ShadowDisableOptions!(execute, Selection:ShadowDisableOptions Enum=Get l_shadowOptions() -> (l_shadowOptions:ShadowDisableOptions Enum=NewEnumerator0)) -> (NewEnumerator0:Exec, NewEnumerator1:Exec, NewEnumerator3:Exec) {is=title K2Node_SwitchEnum}
   > internal_disableShadows(staticMesh=@n16)
   n19 = Get StaticMeshEditorSubsystem()
   n20 = GetLodBuildSettings!(
    execute=@n18.NewEnumerator1
    self:Static Mesh Editor Subsystem Object Reference=@n19
    StaticMesh:Static Mesh Object Reference=@n16
    LodIndex:int=0
   ) -> (then, OutBuildOptions:Mesh Build Settings Structure)
   > n21 = Set members in Mesh Build Settings!(execute:Exec, StructRef:"Mesh Build Settings Structure (by ref)"=@n20, DistanceFieldResolutionScale:float=Get l_scale() -> (l_scale:double)) -> (then:Exec, StructOut:"Mesh Build Settings Structure (by ref)") {is=title K2Node_SetFieldsInStruct}
   > SetLodBuildSettings!(
    execute
    self:Static Mesh Editor Subsystem Object Reference=@n19
    StaticMesh:Static Mesh Object Reference=@n16
    LodIndex:int=0
    BuildOptions:"Mesh Build Settings Structure (by ref)"=@n21
   ) -> (then)
   n22 = Switch on CollisionDisableOptions!(
    execute=@n17.then_1
    Selection:CollisionDisableOptions Enum=Get l_collisionOptions() -> (l_collisionOptions:CollisionDisableOptions Enum=NewEnumerator0)
   ) -> (NewEnumerator0:Exec, NewEnumerator1:Exec, NewEnumerator3:Exec) {is=title K2Node_SwitchEnum}
   > internal_disableCollision(staticMesh=@n16)
   Execute Python Script!(
    execute=@n22.NewEnumerator1
    PythonScript="import unreal\r\n\r\nbody_setup = static_mesh.get_editor_property(\"BodySetup\")\r\nbody_setup.set_editor_property(\"CollisionTraceFlag\", unreal.CollisionTraceFlag.CTF_USE_SIMPLE_AS_COMPLEX)\r\ncomplexity = body_setup.get_editor_property(\"CollisionTraceFlag\")\r\nprint(str(complexity))"
    static_mesh:Static Mesh Object Reference=@n16
   ) -> (then, ReturnValue) {inputs [static_mesh]}
  ]}
 DisableNanite {nodes [
   DisableNanite() {is=title K2Node_FunctionEntry}
   > n1 = GetSelectedAssets()
   > n2 = Filter by Class(TargetArray=@n1, ObjectClass=&StaticMesh)
   > n3 = For Each Loop(Array=@n2)
   n4 = Get StaticMeshEditorSubsystem()
   > n5 = Set members in Mesh Nanite Settings!(
    execute:Exec
    StructRef:"Mesh Nanite Settings Structure (by ref)"=GetNaniteSettings!(self:Static Mesh Editor Subsystem Object Reference=@n4, StaticMesh:Static Mesh Object Reference) -> (ReturnValue:Mesh Nanite Settings Structure)
    bEnabled:bool=False
   ) -> (then:Exec, StructOut:"Mesh Nanite Settings Structure (by ref)") {is=title K2Node_SetFieldsInStruct}
   > SetNaniteSettings!(
    execute
    self:Static Mesh Editor Subsystem Object Reference=@n4
    StaticMesh:Static Mesh Object Reference=@n3
    NaniteSettings:Mesh Nanite Settings Structure=@n5
    bApplyChanges:bool=true
   ) -> (then)
  ]}
 EventGraph {links [], nodes []}
 internal_disableCollision {nodes [
   n1 = internal_disableCollision!() -> (then, staticMesh:Static Mesh Object Reference) {is=title K2Node_FunctionEntry}
   > Set l_mesh(l_mesh:Static Mesh Object Reference=@n1) -> (Output_Get:Static Mesh Object Reference)
   > n2 = Sequence()
   n3 = Get StaticMeshEditorSubsystem()
   > RemoveCollisions!(
    execute
    self:Static Mesh Editor Subsystem Object Reference=@n3
    StaticMesh:Static Mesh Object Reference=Get l_mesh() -> (l_mesh:Static Mesh Object Reference)
   ) -> (then, ReturnValue:bool)
   n4 = GetLodCount!(execute=@n2.then_1, self:Static Mesh Editor Subsystem Object Reference=@n3, StaticMesh:Static Mesh Object Reference=Get l_mesh()) -> (then, ReturnValue:int=0)
   > n5 = For Loop(LastIndex=int - int(A=@n4, B=1) {is=title K2Node_CallFunction})
   n6 = Get l_mesh()
   > n7 = For Loop(LastIndex=int - int(A=GetNumSections!(self:Static Mesh Object Reference=@n6, InLOD:int=@n5) -> (ReturnValue:int=0), B=1))
   > EnableSectionCollision!(
    execute
    self:Static Mesh Editor Subsystem Object Reference=Get StaticMeshEditorSubsystem()
    StaticMesh:Static Mesh Object Reference=Get l_mesh()
    bCollisionEnabled:bool=false
    LODIndex:int=@n5
    SectionIndex:int=@n7
   ) -> (then)
  ]}
 internal_disableShadows {nodes [
   n1 = internal_disableShadows!() -> (then, staticMesh:Static Mesh Object Reference) {is=title K2Node_FunctionEntry}
   > Set l_mesh(l_mesh=@n1)
   > n2 = Sequence()
   n3 = Get StaticMeshEditorSubsystem()
   > n4 = GetLodCount(self=@n3, StaticMesh=Get l_mesh()) -> (ReturnValue=0)
   > n5 = For Loop(LastIndex=int - int(A=@n4, B=1))
   n6 = Get l_mesh()
   > n7 = For Loop(LastIndex=int - int(A=GetNumSections(self=@n6, InLOD=@n5) -> (ReturnValue=0), B=1))
   > EnableSectionCastShadow!(
    execute
    self:Static Mesh Editor Subsystem Object Reference=Get StaticMeshEditorSubsystem()
    StaticMesh:Static Mesh Object Reference=Get l_mesh()
    bCastShadow:bool=false
    LODIndex:int=@n5
    SectionIndex:int=@n7
   ) -> (then)
   n8 = GetLodBuildSettings(execute=@n2.then_1, self=@n3, StaticMesh=Get l_mesh(), LodIndex=0)
   > n9 = Set members in Mesh Build Settings(StructRef=@n8, DistanceFieldResolutionScale=0)
   > SetLodBuildSettings(self=@n3, StaticMesh=Get l_mesh(), LodIndex=0, BuildOptions=@n9)
  ]}
]
variables []
